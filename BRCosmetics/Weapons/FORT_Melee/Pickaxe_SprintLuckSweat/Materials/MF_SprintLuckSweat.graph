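# Material function for the SprintLuckSweat pickaxe that paints an unlit-looking emissive scene inside a sphere mask, blended over incoming material attributes.
# Camera-facing spherical UVs, with a fisheye variant, place three rectangle-masked texture layers, one bobbing on a sine.
# Custom HLSL adds procedurally seeded, blinking star glints to the back layer.
# Fresnel blends the edges toward a cubemap term built from skylight environment samples and a toon-chrome curve.
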MaterialFunction /BRCosmetics/Weapons/FORT_Melee/Pickaxe_SprintLuckSweat/Materials/MF_SprintLuckSweat
graph {expressions {
  MaterialExpressionAbs_0 {class=MaterialExpressionAbs, inputs [None {from=MaterialExpressionMultiply_6, output=""}], properties {}}
  MaterialExpressionAbs_1 {class=MaterialExpressionAbs, inputs [None {from=MaterialExpressionMultiply_17, output=""}], properties {}}
  MaterialExpressionAdd_0 {
   class=MaterialExpressionAdd
   inputs [
    A {from=MaterialExpressionMultiply_20}
    B {from=""}
   ]
   properties {desc=Distortion}
  }
  MaterialExpressionAdd_1 {
   class=MaterialExpressionAdd
   inputs [
    A {from=MaterialExpressionMultiply_22}
    B {from=""}
   ]
   properties {constB=0.5}
  }
  MaterialExpressionAdd_12 {
   class=MaterialExpressionAdd
   inputs [
    A {from=MaterialExpressionTextureSample_2, output=RGB}
    B {from=MaterialExpressionMultiply_48}
   ]
  }
  MaterialExpressionAdd_2 {
   class=MaterialExpressionAdd
   inputs [
    A {from=MaterialExpressionMultiply_31}
    B {from=MaterialExpressionMultiply_32}
   ]
  }
  MaterialExpressionAdd_4 {
   class=MaterialExpressionAdd
   inputs [
    A {from=MaterialExpressionMultiply_40}
    B {from=MaterialExpressionConvert_16}
   ]
  }
  MaterialExpressionAdd_5 {
   class=MaterialExpressionAdd
   inputs [
    A {from=MaterialExpressionNamedRerouteUsage_0}
    B {from=MaterialExpressionConstant2Vector_0}
   ]
  }
  MaterialExpressionAdd_6 {
   class=MaterialExpressionAdd
   inputs [
    A {from=MaterialExpressionMax_0}
    B {from=MaterialExpressionMultiply_18}
   ]
  }
  MaterialExpressionBlendMaterialAttributes_0 {
   class=MaterialExpressionBlendMaterialAttributes
   inputs [
    A {from=MaterialExpressionFunctionInput_0, output=""}
    B {from=MaterialExpressionSetMaterialAttributes_0, output=""}
    Alpha {from=MaterialExpressionSphereMask_0, output=""}
   ]
   properties {}
  }
  MaterialExpressionCameraPositionWS_0 {class=MaterialExpressionCameraPositionWS, inputs [], properties {}}
  MaterialExpressionCameraVectorWS_1 {class=MaterialExpressionCameraVectorWS, inputs [], properties {}}
  MaterialExpressionClamp_0 {
   class=MaterialExpressionClamp
   inputs [
    None {from=MaterialExpressionScalarParameter_20, output=""}
    Min {from="", output=""}
    Max {from="", output=""}
   ]
   properties {maxDefault=10}
  }
  MaterialExpressionComponentMask_1 {class=MaterialExpressionComponentMask, inputs [None {from=MaterialExpressionViewSize_0}], properties {r=true}}
  MaterialExpressionComponentMask_2 {class=MaterialExpressionComponentMask, inputs [None {from=MaterialExpressionViewSize_0}], properties {g=true}}
  MaterialExpressionComponentMask_3 {class=MaterialExpressionComponentMask, inputs [None {from=MaterialExpressionNormalize_1}], properties {b=true}}
  MaterialExpressionConstant2Vector_0 {class=MaterialExpressionConstant2Vector}
  MaterialExpressionConstant3Vector_1 {class=MaterialExpressionConstant3Vector, properties {constant=<rgba 1.072 7.243 53.67 0>}}
  MaterialExpressionConstant3Vector_8 {class=MaterialExpressionConstant3Vector, properties {constant=<rgba 1.072 7.243 53.67 0>}}
  MaterialExpressionConstantBiasScale_1 {class=MaterialExpressionConstantBiasScale, inputs [None {from=MaterialExpressionComponentMask_3}]}
  MaterialExpressionConstant_1 {class=MaterialExpressionConstant}
  MaterialExpressionConstant_10 {class=MaterialExpressionConstant, properties {r=0.5}}
  MaterialExpressionConstant_21 {class=MaterialExpressionConstant, properties {r=0.05}}
  MaterialExpressionConvert_0 {
   class=MaterialExpressionConvert
   inputs [
    None {from=MaterialExpressionVectorParameter_0, output=SubTexMin.X}
    None {from=MaterialExpressionVectorParameter_0, output=SubTexMin.X}
    None {from=MaterialExpressionVectorParameter_0, output=SubTexMin.X}
    None {from=MaterialExpressionVectorParameter_0, output=SubTexMin.X}
   ]
   properties {
    nodeName=Make Vector2
    convertInputs [
     {
      type=Scalar
      expressionInput {
       expression=&self:MaterialExpressionVectorParameter_0
       inputName=None
       mask=1
       maskA=0
       maskB=0
       maskG=0
       maskR=1
       outputIndex=1
      }
     }
     {
      type=Scalar
      expressionInput {
       expression=&self:MaterialExpressionVectorParameter_0
       inputName=None
       mask=1
       maskA=0
       maskB=0
       maskG=1
       maskR=0
       outputIndex=2
      }
     }
     {
      type=Scalar
      expressionInput {
       expression=&self:MaterialExpressionVectorParameter_0
       inputName=None
       mask=1
       maskA=0
       maskB=1
       maskG=0
       maskR=0
       outputIndex=3
      }
     }
     {
      type=Scalar
      expressionInput {
       expression=&self:MaterialExpressionVectorParameter_0
       inputName=None
       mask=1
       maskA=1
       maskB=0
       maskG=0
       maskR=0
       outputIndex=4
      }
     }
    ]
    convertOutputs [{type=Vector2}, {type=Vector2}]
   }
  }
  MaterialExpressionConvert_1 {
   class=MaterialExpressionConvert
   inputs [None {from=MaterialExpressionDivide_0}]
   properties {
    nodeName=Break Float2
    convertInputs [{
      type=Vector2
      expressionInput {
       expression=&self:MaterialExpressionDivide_0
       inputName=None
       mask=0
       maskA=0
       maskB=0
       maskG=0
       maskR=0
       outputIndex=0
      }
     }]
    convertOutputs [{type=Scalar}, {type=Scalar}]
   }
  }
  MaterialExpressionConvert_10 {
   class=MaterialExpressionConvert
   inputs [
    None {from=MaterialExpressionScalarParameter_17}
    None {from=MaterialExpressionScalarParameter_18}
   ]
   properties {
    nodeName=Make Vector2
    convertInputs [
     {
      type=Scalar
      expressionInput {
       expression=&self:MaterialExpressionScalarParameter_17
       inputName=None
       mask=0
       maskA=0
       maskB=0
       maskG=0
       maskR=0
       outputIndex=0
      }
     }
     {
      type=Scalar
      expressionInput {
       expression=&self:MaterialExpressionScalarParameter_18
       inputName=None
       mask=0
       maskA=0
       maskB=0
       maskG=0
       maskR=0
       outputIndex=0
      }
     }
    ]
    convertOutputs [{type=Vector2}]
   }
  }
  MaterialExpressionConvert_11 {
   class=MaterialExpressionConvert
   inputs [None {from=MaterialExpressionAdd_5}]
   properties {
    nodeName=Break Float2
    convertInputs [{
      type=Vector2
      expressionInput {
       expression=&self:MaterialExpressionAdd_5
       inputName=None
       mask=0
       maskA=0
       maskB=0
       maskG=0
       maskR=0
       outputIndex=0
      }
     }]
    convertOutputs [{type=Scalar}, {type=Scalar}]
   }
  }
  MaterialExpressionConvert_16 {
   class=MaterialExpressionConvert
   inputs [None {from=MaterialExpressionSubtract_5}]
   properties {
    nodeName=Break Float2
    convertInputs [{
      type=Vector2
      expressionInput {
       expression=&self:MaterialExpressionSubtract_5
       inputName=None
       mask=0
       maskA=0
       maskB=0
       maskG=0
       maskR=0
       outputIndex=0
      }
     }]
    convertOutputs [{type=Scalar}, {type=Scalar}]
   }
  }
  MaterialExpressionConvert_17 {
   class=MaterialExpressionConvert
   inputs [
    None {from=MaterialExpressionConvert_16}
    None {from=MaterialExpressionAdd_4}
   ]
   properties {
    nodeName=Make Vector2
    convertInputs [
     {
      type=Scalar
      expressionInput {
       expression=&self:MaterialExpressionConvert_16
       inputName=None
       mask=0
       maskA=0
       maskB=0
       maskG=0
       maskR=0
       outputIndex=0
      }
     }
     {
      type=Scalar
      expressionInput {
       expression=&self:MaterialExpressionAdd_4
       inputName=None
       mask=0
       maskA=0
       maskB=0
       maskG=0
       maskR=0
       outputIndex=0
      }
     }
    ]
    convertOutputs [{type=Vector2}]
   }
  }
  MaterialExpressionConvert_2 {
   class=MaterialExpressionConvert
   inputs [
    None {from=MaterialExpressionVectorParameter_1, output=SubTexMin.X}
    None {from=MaterialExpressionVectorParameter_1, output=SubTexMin.X}
    None {from=MaterialExpressionVectorParameter_1, output=SubTexMin.X}
    None {from=MaterialExpressionVectorParameter_1, output=SubTexMin.X}
   ]
   properties {
    nodeName=Make Vector2
    convertInputs [
     {
      type=Scalar
      expressionInput {
       expression=&self:MaterialExpressionVectorParameter_1
       inputName=None
       mask=1
       maskA=0
       maskB=0
       maskG=0
       maskR=1
       outputIndex=1
      }
     }
     {
      type=Scalar
      expressionInput {
       expression=&self:MaterialExpressionVectorParameter_1
       inputName=None
       mask=1
       maskA=0
       maskB=0
       maskG=1
       maskR=0
       outputIndex=2
      }
     }
     {
      type=Scalar
      expressionInput {
       expression=&self:MaterialExpressionVectorParameter_1
       inputName=None
       mask=1
       maskA=0
       maskB=1
       maskG=0
       maskR=0
       outputIndex=3
      }
     }
     {
      type=Scalar
      expressionInput {
       expression=&self:MaterialExpressionVectorParameter_1
       inputName=None
       mask=1
       maskA=1
       maskB=0
       maskG=0
       maskR=0
       outputIndex=4
      }
     }
    ]
    convertOutputs [{type=Vector2}, {type=Vector2}]
   }
  }
  MaterialExpressionConvert_3 {
   class=MaterialExpressionConvert
   inputs [None {from=MaterialExpressionDivide_1}]
   properties {
    nodeName=Break Float2
    convertInputs [{
      type=Vector2
      expressionInput {
       expression=&self:MaterialExpressionDivide_1
       inputName=None
       mask=0
       maskA=0
       maskB=0
       maskG=0
       maskR=0
       outputIndex=0
      }
     }]
    convertOutputs [{type=Scalar}, {type=Scalar}]
   }
  }
  MaterialExpressionConvert_4 {
   class=MaterialExpressionConvert
   inputs [None {from=MaterialExpressionDivide_2}]
   properties {
    nodeName=Break Float2
    convertInputs [{
      type=Vector2
      expressionInput {
       expression=&self:MaterialExpressionDivide_2
       inputName=None
       mask=0
       maskA=0
       maskB=0
       maskG=0
       maskR=0
       outputIndex=0
      }
     }]
    convertOutputs [{type=Scalar}, {type=Scalar}]
   }
  }
  MaterialExpressionConvert_5 {
   class=MaterialExpressionConvert
   inputs [
    None {from=MaterialExpressionVectorParameter_2, output=SubTexMin.X}
    None {from=MaterialExpressionVectorParameter_2, output=SubTexMin.X}
    None {from=MaterialExpressionVectorParameter_2, output=SubTexMin.X}
    None {from=MaterialExpressionVectorParameter_2, output=SubTexMin.X}
   ]
   properties {
    nodeName=Make Vector2
    convertInputs [
     {
      type=Scalar
      expressionInput {
       expression=&self:MaterialExpressionVectorParameter_2
       inputName=None
       mask=1
       maskA=0
       maskB=0
       maskG=0
       maskR=1
       outputIndex=1
      }
     }
     {
      type=Scalar
      expressionInput {
       expression=&self:MaterialExpressionVectorParameter_2
       inputName=None
       mask=1
       maskA=0
       maskB=0
       maskG=1
       maskR=0
       outputIndex=2
      }
     }
     {
      type=Scalar
      expressionInput {
       expression=&self:MaterialExpressionVectorParameter_2
       inputName=None
       mask=1
       maskA=0
       maskB=1
       maskG=0
       maskR=0
       outputIndex=3
      }
     }
     {
      type=Scalar
      expressionInput {
       expression=&self:MaterialExpressionVectorParameter_2
       inputName=None
       mask=1
       maskA=1
       maskB=0
       maskG=0
       maskR=0
       outputIndex=4
      }
     }
    ]
    convertOutputs [{type=Vector2}, {type=Vector2}]
   }
  }
  MaterialExpressionConvert_6 {
   class=MaterialExpressionConvert
   inputs [
    None {from=MaterialExpressionScalarParameter_5}
    None {from=MaterialExpressionScalarParameter_8}
   ]
   properties {
    nodeName=Make Vector2
    convertInputs [
     {
      type=Scalar
      expressionInput {
       expression=&self:MaterialExpressionScalarParameter_5
       inputName=None
       mask=0
       maskA=0
       maskB=0
       maskG=0
       maskR=0
       outputIndex=0
      }
     }
     {
      type=Scalar
      expressionInput {
       expression=&self:MaterialExpressionScalarParameter_8
       inputName=None
       mask=0
       maskA=0
       maskB=0
       maskG=0
       maskR=0
       outputIndex=0
      }
     }
    ]
    convertOutputs [{type=Vector2}]
   }
  }
  MaterialExpressionConvert_7 {
   class=MaterialExpressionConvert
   inputs [
    None {from=MaterialExpressionScalarParameter_14}
    None {from=MaterialExpressionScalarParameter_15}
   ]
   properties {
    nodeName=Make Vector2
    convertInputs [
     {
      type=Scalar
      expressionInput {
       expression=&self:MaterialExpressionScalarParameter_14
       inputName=None
       mask=0
       maskA=0
       maskB=0
       maskG=0
       maskR=0
       outputIndex=0
      }
     }
     {
      type=Scalar
      expressionInput {
       expression=&self:MaterialExpressionScalarParameter_15
       inputName=None
       mask=0
       maskA=0
       maskB=0
       maskG=0
       maskR=0
       outputIndex=0
      }
     }
    ]
    convertOutputs [{type=Vector2}]
   }
  }
  MaterialExpressionCurveAtlasRowParameter_2 {
   class=MaterialExpressionCurveAtlasRowParameter
   inputs [CurveTime {from=MaterialExpressionConstantBiasScale_1, output=""}]
   properties {
    atlas=/Game/Global/Materials/Character_Parent_2023/Comic_Parent_2023/Textures/CA_ToonChromeAtlas
    curve=/Game/Global/Materials/Character_Parent_2023/Comic_Parent_2023/Textures/C_ToonChrome_01
    group=SprintLuckSweat
    parameterName=ChromeCurve
   }
  }
  MaterialExpressionCustom_0 {
   class=MaterialExpressionCustom
   inputs [
    cameraPos {from=MaterialExpressionSubtract_16, output=""}
    fragmentPos {from=MaterialExpressionSubtract_17, output=""}
    scale {from=MaterialExpressionConstant_21, output=""}
   ]
   properties {
    code="float3 w = -fragmentPos;\r\nfloat3 forward = normalize(-cameraPos);\r\nfloat3 up = float3(0, 0, 1);\r\nfloat3 side = normalize(cross(forward, up));\r\nup = cross(side, forward);\r\nfloat u = scale * dot(w, side) + 0.5;\r\nfloat v = scale * dot(w, up) + 0.5;\r\nuvs = float2(u, v);\r\nreturn up;"
    description=ViewSpaceSphereicalUV
    showCode=true
    additionalOutputs [{outputName=uvs, outputType=CMOT_Float2}]
    inputs [
     {
      inputName=cameraPos
      input {
       expression=&self:MaterialExpressionSubtract_16
       inputName=None
       mask=0
       maskA=0
       maskB=0
       maskG=0
       maskR=0
       outputIndex=0
      }
     }
     {
      inputName=fragmentPos
      input {
       expression=&self:MaterialExpressionSubtract_17
       inputName=None
       mask=0
       maskA=0
       maskB=0
       maskG=0
       maskR=0
       outputIndex=0
      }
     }
     {
      inputName=scale
      input {
       expression=&self:MaterialExpressionConstant_21
       inputName=None
       mask=0
       maskA=0
       maskB=0
       maskG=0
       maskR=0
       outputIndex=0
      }
     }
    ]
   }
  }
  MaterialExpressionCustom_2 {
   class=MaterialExpressionCustom
   inputs [
    UV {from=MaterialExpressionNamedRerouteUsage_3, output=""}
    StarCount {from=MaterialExpressionClamp_0, output=""}
    Time {from=MaterialExpressionTime_2, output=""}
    Speed {from=MaterialExpressionScalarParameter_21, output=""}
    SizeMin {from=MaterialExpressionScalarParameter_23, output=""}
    SizeMax {from=MaterialExpressionScalarParameter_24, output=""}
    Seed {from=MaterialExpressionScalarParameter_22, output=""}
   ]
   properties {
    code="float total = 0.0;\r\nint count = (int)StarCount;\r\n\r\nfor (int i = 0; i < count; i++)\r\n{\r\n    \r\n    float fi = (float)i + Seed * 13.57;\r\n    float2 starPos = float2(\r\n        frac(sin(fi * 127.1 + 0.5) * 43758.5453),\r\n        frac(sin(fi * 311.7 + 1.3) * 43758.5453)\r\n    );\r\n    float r    = frac(sin(fi * 93.989) * 43758.5453);   \r\n    float size = lerp(SizeMin, SizeMax, r);\r\n\r\n    float2 uv = (UV - starPos) / size;\r\n\r\n    float sharp = max(0.0, 1.0 - abs(uv.x * uv.y * 1000.0));\r\n    float glow  = max(0.0, 1.0 - abs(uv.x * uv.y * 50.0));\r\n    float mask  = smoothstep(0.65, 0.0, length(uv));\r\n    float star  = pow((sharp + glow * 0.5) * mask,3.0);\r\n\r\n    float phase      = frac(sin(fi * 78.233) * 43758.5453) * 6.2831853;\r\n    float blinkSpeed = 0.5 + frac(sin(fi * 45.164) * 43758.5453) * 1.5;\r\n    float blink      = 0.5 + 0.5 * sin(Time * Speed * blinkSpeed + phase);\r\n\r\n    total += star * blink;\r\n}\r\n\r\nreturn total;"
    outputType=CMOT_Float1
    showCode=true
    inputs [
     {
      inputName=UV
      input {
       expression=&self:MaterialExpressionNamedRerouteUsage_3
       inputName=None
       mask=0
       maskA=0
       maskB=0
       maskG=0
       maskR=0
       outputIndex=0
      }
     }
     {
      inputName=StarCount
      input {
       expression=&self:MaterialExpressionClamp_0
       inputName=None
       mask=0
       maskA=0
       maskB=0
       maskG=0
       maskR=0
       outputIndex=0
      }
     }
     {
      inputName=Time
      input {
       expression=&self:MaterialExpressionTime_2
       inputName=None
       mask=0
       maskA=0
       maskB=0
       maskG=0
       maskR=0
       outputIndex=0
      }
     }
     {
      inputName=Speed
      input {
       expression=&self:MaterialExpressionScalarParameter_21
       inputName=None
       mask=0
       maskA=0
       maskB=0
       maskG=0
       maskR=0
       outputIndex=0
      }
     }
     {
      inputName=SizeMin
      input {
       expression=&self:MaterialExpressionScalarParameter_23
       inputName=None
       mask=0
       maskA=0
       maskB=0
       maskG=0
       maskR=0
       outputIndex=0
      }
     }
     {
      inputName=SizeMax
      input {
       expression=&self:MaterialExpressionScalarParameter_24
       inputName=None
       mask=0
       maskA=0
       maskB=0
       maskG=0
       maskR=0
       outputIndex=0
      }
     }
     {
      inputName=Seed
      input {
       expression=&self:MaterialExpressionScalarParameter_22
       inputName=None
       mask=0
       maskA=0
       maskB=0
       maskG=0
       maskR=0
       outputIndex=0
      }
     }
    ]
   }
  }
  MaterialExpressionDesaturation_0 {
   class=MaterialExpressionDesaturation
   inputs [
    None {from=MaterialExpressionMultiply_35, output=""}
    Fraction {from=MaterialExpressionScalarParameter_31, output=""}
   ]
   properties {}
  }
  MaterialExpressionDivide_0 {
   class=MaterialExpressionDivide
   inputs [
    A {from=MaterialExpressionSubtract_3}
    B {from=MaterialExpressionSubtract_4}
   ]
  }
  MaterialExpressionDivide_1 {
   class=MaterialExpressionDivide
   inputs [
    A {from=MaterialExpressionSubtract_6}
    B {from=MaterialExpressionSubtract_7}
   ]
  }
  MaterialExpressionDivide_2 {
   class=MaterialExpressionDivide
   inputs [
    A {from=MaterialExpressionSubtract_0}
    B {from=MaterialExpressionSubtract_1}
   ]
  }
  MaterialExpressionDivide_5 {
   class=MaterialExpressionDivide
   inputs [
    A {from=MaterialExpressionMultiply_21}
    B {from=MaterialExpressionScalarParameter_7}
   ]
  }
  MaterialExpressionDivide_8 {
   class=MaterialExpressionDivide
   inputs [
    A {from=MaterialExpressionComponentMask_1}
    B {from=MaterialExpressionComponentMask_2}
   ]
  }
  MaterialExpressionDotProduct_0 {
   class=MaterialExpressionDotProduct
   inputs [
    A {from=MaterialExpressionMultiply_19, output=""}
    B {from=MaterialExpressionMultiply_19, output=""}
   ]
   properties {}
  }
  MaterialExpressionFresnel_0 {
   class=MaterialExpressionFresnel
   inputs [
    ExponentIn {from="", output=""}
    BaseReflectFractionIn {from="", output=""}
    Normal {from="", output=""}
   ]
   properties {exponent=2}
  }
  MaterialExpressionFunctionInput_0 {
   class=MaterialExpressionFunctionInput
   inputs [Preview {from="", output=""}]
   properties {inputName=IN MA, inputType=FunctionInput_MaterialAttributes}
  }
  MaterialExpressionFunctionOutput_0 {class=MaterialExpressionFunctionOutput, inputs [None {from=MaterialExpressionBlendMaterialAttributes_0, output=""}]}
  MaterialExpressionLength_0 {class=MaterialExpressionLength, inputs [None {from=MaterialExpressionAdd_5, output=""}], properties {}}
  MaterialExpressionLinearInterpolate_0 {
   class=MaterialExpressionLinearInterpolate
   inputs [
    A {from=MaterialExpressionLinearInterpolate_1}
    B {from=MaterialExpressionMultiply_0}
    Alpha {from=MaterialExpressionMultiply_25}
   ]
  }
  MaterialExpressionLinearInterpolate_1 {
   class=MaterialExpressionLinearInterpolate
   inputs [
    A {from=MaterialExpressionMultiply_13}
    B {from=MaterialExpressionMultiply_26}
    Alpha {from=MaterialExpressionMultiply_5}
   ]
  }
  MaterialExpressionLinearInterpolate_9 {
   class=MaterialExpressionLinearInterpolate
   inputs [
    A {from=MaterialExpressionLinearInterpolate_0}
    B {from=MaterialExpressionNamedRerouteUsage_12}
    Alpha {from=MaterialExpressionFresnel_0}
   ]
  }
  MaterialExpressionMax_0 {
   class=MaterialExpressionMax
   inputs [
    A {from=""}
    B {from=MaterialExpressionOneMinus_0}
   ]
   properties {desc=Sharp}
  }
  MaterialExpressionMax_1 {
   class=MaterialExpressionMax
   inputs [
    A {from=""}
    B {from=MaterialExpressionOneMinus_1}
   ]
   properties {desc=Glow}
  }
  MaterialExpressionMultiply_0 {
   class=MaterialExpressionMultiply
   inputs [
    A {from=MaterialExpressionTextureSample_0, output=RGB}
    B {from=MaterialExpressionMultiply_25}
   ]
  }
  MaterialExpressionMultiply_1 {
   class=MaterialExpressionMultiply
   inputs [
    A {from=MaterialExpressionConvert_11}
    B {from=MaterialExpressionConvert_11}
   ]
  }
  MaterialExpressionMultiply_10 {
   class=MaterialExpressionMultiply
   inputs [
    A {from=MaterialExpressionMultiply_7}
    B {from=MaterialExpressionMultiply_9}
   ]
   properties {desc=RectangleMask}
  }
  MaterialExpressionMultiply_11 {
   class=MaterialExpressionMultiply
   inputs [
    A {from=MaterialExpressionStep_10}
    B {from=MaterialExpressionStep_11}
   ]
  }
  MaterialExpressionMultiply_12 {
   class=MaterialExpressionMultiply
   inputs [
    A {from=MaterialExpressionMultiply_8}
    B {from=MaterialExpressionMultiply_11}
   ]
   properties {desc=RectangleMask}
  }
  MaterialExpressionMultiply_13 {
   class=MaterialExpressionMultiply
   inputs [
    A {from=MaterialExpressionAdd_12}
    B {from=MaterialExpressionMultiply_24}
   ]
  }
  MaterialExpressionMultiply_14 {
   class=MaterialExpressionMultiply
   inputs [
    A {from=MaterialExpressionNamedRerouteUsage_8}
    B {from=""}
   ]
   properties {constB=2}
  }
  MaterialExpressionMultiply_15 {
   class=MaterialExpressionMultiply
   inputs [
    A {from=MaterialExpressionNamedRerouteUsage_4}
    B {from=MaterialExpressionScalarParameter_16}
   ]
   properties {constB=0.4}
  }
  MaterialExpressionMultiply_17 {
   class=MaterialExpressionMultiply
   inputs [
    A {from=MaterialExpressionMultiply_1}
    B {from=""}
   ]
   properties {constB=50}
  }
  MaterialExpressionMultiply_18 {
   class=MaterialExpressionMultiply
   inputs [
    A {from=MaterialExpressionMax_1}
    B {from=""}
   ]
   properties {constB=0.1}
  }
  MaterialExpressionMultiply_19 {
   class=MaterialExpressionMultiply
   inputs [
    A {from=MaterialExpressionSubtract_11}
    B {from=MaterialExpressionScalarParameter_7}
   ]
  }
  MaterialExpressionMultiply_2 {
   class=MaterialExpressionMultiply
   inputs [
    A {from=MaterialExpressionStep_0}
    B {from=MaterialExpressionStep_1}
   ]
  }
  MaterialExpressionMultiply_20 {
   class=MaterialExpressionMultiply
   inputs [
    A {from=MaterialExpressionDotProduct_0}
    B {from=MaterialExpressionScalarParameter_19}
   ]
  }
  MaterialExpressionMultiply_21 {
   class=MaterialExpressionMultiply
   inputs [
    A {from=MaterialExpressionAdd_0}
    B {from=MaterialExpressionMultiply_19}
   ]
  }
  MaterialExpressionMultiply_22 {
   class=MaterialExpressionMultiply
   inputs [
    A {from=MaterialExpressionDivide_5}
    B {from=""}
   ]
   properties {constB=0.5}
  }
  MaterialExpressionMultiply_23 {
   class=MaterialExpressionMultiply
   inputs [
    A {from=MaterialExpressionAdd_6}
    B {from=MaterialExpressionSmoothStep_0}
   ]
  }
  MaterialExpressionMultiply_24 {
   class=MaterialExpressionMultiply
   inputs [
    A {from=MaterialExpressionTextureSample_2, output=A}
    B {from=MaterialExpressionMultiply_12}
   ]
  }
  MaterialExpressionMultiply_25 {
   class=MaterialExpressionMultiply
   inputs [
    A {from=MaterialExpressionTextureSample_0, output=A}
    B {from=MaterialExpressionMultiply_10}
   ]
  }
  MaterialExpressionMultiply_26 {
   class=MaterialExpressionMultiply
   inputs [
    A {from=MaterialExpressionTextureSample_1, output=RGB}
    B {from=MaterialExpressionMultiply_5}
   ]
  }
  MaterialExpressionMultiply_3 {
   class=MaterialExpressionMultiply
   inputs [
    A {from=MaterialExpressionStep_6}
    B {from=MaterialExpressionStep_7}
   ]
  }
  MaterialExpressionMultiply_31 {
   class=MaterialExpressionMultiply
   inputs [
    A {from=MaterialExpressionCameraVectorWS_1}
    B {from=""}
   ]
   properties {constB=-1}
  }
  MaterialExpressionMultiply_32 {
   class=MaterialExpressionMultiply
   inputs [
    A {from=MaterialExpressionVertexNormalWS_0}
    B {from=MaterialExpressionConstant_10}
   ]
  }
  MaterialExpressionMultiply_33 {
   class=MaterialExpressionMultiply
   inputs [
    A {from=MaterialExpressionSkyLightEnvMapSample_0}
    B {from=MaterialExpressionDesaturation_0}
   ]
  }
  MaterialExpressionMultiply_35 {
   class=MaterialExpressionMultiply
   inputs [
    A {from=MaterialExpressionPower_5}
    B {from=MaterialExpressionScalarParameter_30}
   ]
  }
  MaterialExpressionMultiply_36 {
   class=MaterialExpressionMultiply
   inputs [
    A {from=MaterialExpressionDesaturation_0}
    B {from=MaterialExpressionOneMinus_5}
   ]
   properties {constB=0.1}
  }
  MaterialExpressionMultiply_37 {
   class=MaterialExpressionMultiply
   inputs [
    A {from=MaterialExpressionNamedRerouteUsage_5}
    B {from=MaterialExpressionScalarParameter_0}
   ]
  }
  MaterialExpressionMultiply_38 {
   class=MaterialExpressionMultiply
   inputs [
    A {from=MaterialExpressionTime_0}
    B {from=""}
   ]
   properties {constB=0.2}
  }
  MaterialExpressionMultiply_39 {
   class=MaterialExpressionMultiply
   inputs [
    A {from=MaterialExpressionNamedRerouteUsage_10}
    B {from=MaterialExpressionScalarParameter_9}
   ]
   properties {constB=0.5}
  }
  MaterialExpressionMultiply_4 {
   class=MaterialExpressionMultiply
   inputs [
    A {from=MaterialExpressionMultiply_2}
    B {from=MaterialExpressionMultiply_3}
   ]
   properties {desc=RectangleMask}
  }
  MaterialExpressionMultiply_40 {
   class=MaterialExpressionMultiply
   inputs [
    A {from=MaterialExpressionSine_0}
    B {from=""}
   ]
   properties {constB=0.025}
  }
  MaterialExpressionMultiply_47 {
   class=MaterialExpressionMultiply
   inputs [
    A {from=MaterialExpressionTextureSample_2, output=RGB}
    B {from=MaterialExpressionCustom_2}
   ]
  }
  MaterialExpressionMultiply_48 {
   class=MaterialExpressionMultiply
   inputs [
    A {from=MaterialExpressionMultiply_47}
    B {from=""}
   ]
   properties {constB=10}
  }
  MaterialExpressionMultiply_5 {
   class=MaterialExpressionMultiply
   inputs [
    A {from=MaterialExpressionTextureSample_1, output=A}
    B {from=MaterialExpressionMultiply_4}
   ]
  }
  MaterialExpressionMultiply_6 {
   class=MaterialExpressionMultiply
   inputs [
    A {from=MaterialExpressionMultiply_1}
    B {from=""}
   ]
   properties {constB=1000}
  }
  MaterialExpressionMultiply_7 {
   class=MaterialExpressionMultiply
   inputs [
    A {from=MaterialExpressionStep_2}
    B {from=MaterialExpressionStep_3}
   ]
  }
  MaterialExpressionMultiply_8 {
   class=MaterialExpressionMultiply
   inputs [
    A {from=MaterialExpressionStep_8}
    B {from=MaterialExpressionStep_9}
   ]
  }
  MaterialExpressionMultiply_9 {
   class=MaterialExpressionMultiply
   inputs [
    A {from=MaterialExpressionStep_4}
    B {from=MaterialExpressionStep_5}
   ]
  }
  MaterialExpressionNamedRerouteDeclaration_0 {
   class=MaterialExpressionNamedRerouteDeclaration
   inputs [None {from=MaterialExpressionCustom_0, output=uvs}]
   properties {name=SphereicalUV, nodeColor=<rgba 0 0.3438 1 1>}
  }
  MaterialExpressionNamedRerouteDeclaration_1 {
   class=MaterialExpressionNamedRerouteDeclaration
   inputs [None {from=MaterialExpressionAdd_1}]
   properties {name=SphereicalUV_Fisheye, nodeColor=<rgba 0 0.3438 1 1>}
  }
  MaterialExpressionNamedRerouteDeclaration_2 {
   class=MaterialExpressionNamedRerouteDeclaration
   inputs [None {from=MaterialExpressionMultiply_33}]
   properties {name=Cubemap, nodeColor=<rgba 1 0 0.7266 1>}
  }
  MaterialExpressionNamedRerouteDeclaration_3 {
   class=MaterialExpressionNamedRerouteDeclaration
   inputs [None {from=MaterialExpressionMultiply_19}]
   properties {name=StarStableUV, nodeColor=<rgba 0 0.5547 1 1>}
  }
  MaterialExpressionNamedRerouteUsage_0 {class=MaterialExpressionNamedRerouteUsage}
  MaterialExpressionNamedRerouteUsage_1 {class=MaterialExpressionNamedRerouteUsage}
  MaterialExpressionNamedRerouteUsage_10 {class=MaterialExpressionNamedRerouteUsage}
  MaterialExpressionNamedRerouteUsage_12 {class=MaterialExpressionNamedRerouteUsage}
  MaterialExpressionNamedRerouteUsage_13 {class=MaterialExpressionNamedRerouteUsage}
  MaterialExpressionNamedRerouteUsage_3 {class=MaterialExpressionNamedRerouteUsage}
  MaterialExpressionNamedRerouteUsage_4 {class=MaterialExpressionNamedRerouteUsage}
  MaterialExpressionNamedRerouteUsage_5 {class=MaterialExpressionNamedRerouteUsage}
  MaterialExpressionNamedRerouteUsage_8 {class=MaterialExpressionNamedRerouteUsage}
  MaterialExpressionNormalize_1 {class=MaterialExpressionNormalize, inputs [VectorInput {from=MaterialExpressionReflectionVectorWS_1, output=""}]}
  MaterialExpressionOneMinus_0 {class=MaterialExpressionOneMinus, inputs [None {from=MaterialExpressionAbs_0}]}
  MaterialExpressionOneMinus_1 {class=MaterialExpressionOneMinus, inputs [None {from=MaterialExpressionAbs_1}]}
  MaterialExpressionOneMinus_5 {class=MaterialExpressionOneMinus, inputs [None {from=MaterialExpressionNamedRerouteUsage_13}]}
  MaterialExpressionPower_2 {
   class=MaterialExpressionPower
   inputs [
    Base {from=MaterialExpressionMultiply_23}
    Exp {from=""}
   ]
   properties {constExponent=3}
  }
  MaterialExpressionPower_5 {
   class=MaterialExpressionPower
   inputs [
    Base {from=MaterialExpressionCurveAtlasRowParameter_2}
    Exp {from=MaterialExpressionScalarParameter_26}
   ]
  }
  MaterialExpressionReflectionVectorWS_1 {class=MaterialExpressionReflectionVectorWS, inputs [CustomWorldNormal {from="", output=""}], properties {}}
  MaterialExpressionScalarParameter_0 {class=MaterialExpressionScalarParameter, properties {defaultValue=0.25, group=SprintLuckSweat, parameterName=TextureA_Scale}}
  MaterialExpressionScalarParameter_12 {class=MaterialExpressionScalarParameter, properties {defaultValue=0.8, group=SprintLuckSweat, parameterName=SkylightEnvMapRoughness}}
  MaterialExpressionScalarParameter_14 {class=MaterialExpressionScalarParameter, properties {defaultValue=-0.5, group=SprintLuckSweat, parameterName=TextureB_OffsetX}}
  MaterialExpressionScalarParameter_15 {class=MaterialExpressionScalarParameter, properties {group=SprintLuckSweat, parameterName=TextureB_OffsetY}}
  MaterialExpressionScalarParameter_16 {class=MaterialExpressionScalarParameter, properties {defaultValue=0.4, group=SprintLuckSweat, parameterName=TextureC_Scale}}
  MaterialExpressionScalarParameter_17 {class=MaterialExpressionScalarParameter, properties {defaultValue=-0.55, group=SprintLuckSweat, parameterName=TextureC_OffsetX}}
  MaterialExpressionScalarParameter_18 {class=MaterialExpressionScalarParameter, properties {defaultValue=-0.55, group=SprintLuckSweat, parameterName=TextureC_OffsetY}}
  MaterialExpressionScalarParameter_19 {class=MaterialExpressionScalarParameter, properties {defaultValue=1, group=SprintLuckSweat, parameterName=Fisheye_Strengh}}
  MaterialExpressionScalarParameter_20 {class=MaterialExpressionScalarParameter, properties {defaultValue=4, group=SprintLuckSweat, parameterName=Stars}}
  MaterialExpressionScalarParameter_21 {class=MaterialExpressionScalarParameter, properties {defaultValue=1, group=SprintLuckSweat, parameterName=Stars_Speed}}
  MaterialExpressionScalarParameter_22 {class=MaterialExpressionScalarParameter, properties {defaultValue=16.6767, group=SprintLuckSweat, parameterName=Stars_Seed}}
  MaterialExpressionScalarParameter_23 {class=MaterialExpressionScalarParameter, properties {defaultValue=0.1, group=SprintLuckSweat, parameterName=Stars_SizeMin}}
  MaterialExpressionScalarParameter_24 {class=MaterialExpressionScalarParameter, properties {defaultValue=1, group=SprintLuckSweat, parameterName=Stars_SizeMax}}
  MaterialExpressionScalarParameter_26 {class=MaterialExpressionScalarParameter, properties {defaultValue=1.2, group=SprintLuckSweat, parameterName=ChromeExponent}}
  MaterialExpressionScalarParameter_30 {class=MaterialExpressionScalarParameter, properties {defaultValue=1, group=SprintLuckSweat, parameterName=ChromeMultiplier}}
  MaterialExpressionScalarParameter_31 {class=MaterialExpressionScalarParameter, properties {group=SprintLuckSweat, parameterName=ChromeDesaturation}}
  MaterialExpressionScalarParameter_5 {class=MaterialExpressionScalarParameter, properties {defaultValue=-0.125, group=SprintLuckSweat, parameterName=TextureA_OffsetX}}
  MaterialExpressionScalarParameter_7 {class=MaterialExpressionScalarParameter, properties {defaultValue=0.9, group=SprintLuckSweat, parameterName=Fisheye_Aspect}}
  MaterialExpressionScalarParameter_8 {class=MaterialExpressionScalarParameter, properties {defaultValue=-0.5, group=SprintLuckSweat, parameterName=TextureA_OffsetY}}
  MaterialExpressionScalarParameter_9 {class=MaterialExpressionScalarParameter, properties {defaultValue=0.4, group=SprintLuckSweat, parameterName=TextureB_Scale}}
  MaterialExpressionSetMaterialAttributes_0 {
   class=MaterialExpressionSetMaterialAttributes
   inputs [
    MaterialAttributes {from=MaterialExpressionFunctionInput_0, output=""}
    BaseColor {from=MaterialExpressionConstant_1, output=""}
    EmissiveColor {from=MaterialExpressionLinearInterpolate_9, output=""}
   ]
   properties {attributeSetTypes [#1, #2]}
  }
  MaterialExpressionSine_0 {class=MaterialExpressionSine, inputs [None {from=MaterialExpressionMultiply_38}]}
  MaterialExpressionSkyLightEnvMapSample_0 {
   class=MaterialExpressionSkyLightEnvMapSample
   inputs [
    Direction {from=MaterialExpressionAdd_2, output=""}
    Roughness {from=MaterialExpressionScalarParameter_12, output=""}
   ]
   properties {}
  }
  MaterialExpressionSmoothStep_0 {
   class=MaterialExpressionSmoothStep
   inputs [
    Min {from="", output=""}
    Max {from="", output=""}
    Value {from=MaterialExpressionLength_0, output=""}
   ]
   properties {constMax=0, constMin=0.5}
  }
  MaterialExpressionSphereMask_0 {
   class=MaterialExpressionSphereMask
   inputs [
    A {from=MaterialExpressionWorldPosition_0, output=XYZ}
    B {from=MaterialExpressionTransformPosition_1, output=""}
    Radius {from="", output=""}
    Hardness {from="", output=""}
   ]
   properties {attenuationRadius=13.25}
  }
  MaterialExpressionStep_0 {
   class=MaterialExpressionStep
   inputs [
    Y {from="", output=""}
    X {from=MaterialExpressionConvert_1, output=""}
   ]
   properties {constX=0}
  }
  MaterialExpressionStep_1 {
   class=MaterialExpressionStep
   inputs [
    Y {from=MaterialExpressionConvert_1, output=""}
    X {from="", output=""}
   ]
   properties {}
  }
  MaterialExpressionStep_10 {
   class=MaterialExpressionStep
   inputs [
    Y {from="", output=""}
    X {from=MaterialExpressionConvert_3, output=""}
   ]
   properties {constX=0}
  }
  MaterialExpressionStep_11 {
   class=MaterialExpressionStep
   inputs [
    Y {from=MaterialExpressionConvert_3, output=""}
    X {from="", output=""}
   ]
   properties {}
  }
  MaterialExpressionStep_2 {
   class=MaterialExpressionStep
   inputs [
    Y {from="", output=""}
    X {from=MaterialExpressionConvert_4, output=""}
   ]
   properties {constX=0}
  }
  MaterialExpressionStep_3 {
   class=MaterialExpressionStep
   inputs [
    Y {from=MaterialExpressionConvert_4, output=""}
    X {from="", output=""}
   ]
   properties {}
  }
  MaterialExpressionStep_4 {
   class=MaterialExpressionStep
   inputs [
    Y {from="", output=""}
    X {from=MaterialExpressionConvert_4, output=""}
   ]
   properties {constX=0}
  }
  MaterialExpressionStep_5 {
   class=MaterialExpressionStep
   inputs [
    Y {from=MaterialExpressionConvert_4, output=""}
    X {from="", output=""}
   ]
   properties {}
  }
  MaterialExpressionStep_6 {
   class=MaterialExpressionStep
   inputs [
    Y {from="", output=""}
    X {from=MaterialExpressionConvert_1, output=""}
   ]
   properties {constX=0}
  }
  MaterialExpressionStep_7 {
   class=MaterialExpressionStep
   inputs [
    Y {from=MaterialExpressionConvert_1, output=""}
    X {from="", output=""}
   ]
   properties {}
  }
  MaterialExpressionStep_8 {
   class=MaterialExpressionStep
   inputs [
    Y {from="", output=""}
    X {from=MaterialExpressionConvert_3, output=""}
   ]
   properties {constX=0}
  }
  MaterialExpressionStep_9 {
   class=MaterialExpressionStep
   inputs [
    Y {from=MaterialExpressionConvert_3, output=""}
    X {from="", output=""}
   ]
   properties {}
  }
  MaterialExpressionSubtract_0 {
   class=MaterialExpressionSubtract
   inputs [
    A {from=MaterialExpressionSubtract_2}
    B {from=MaterialExpressionConvert_0}
   ]
  }
  MaterialExpressionSubtract_1 {
   class=MaterialExpressionSubtract
   inputs [
    A {from=MaterialExpressionConvert_0}
    B {from=MaterialExpressionConvert_0}
   ]
  }
  MaterialExpressionSubtract_11 {
   class=MaterialExpressionSubtract
   inputs [
    A {from=MaterialExpressionMultiply_14}
    B {from=""}
   ]
  }
  MaterialExpressionSubtract_16 {
   class=MaterialExpressionSubtract
   inputs [
    A {from=MaterialExpressionCameraPositionWS_0}
    B {from=MaterialExpressionTransformPosition_0}
   ]
  }
  MaterialExpressionSubtract_17 {
   class=MaterialExpressionSubtract
   inputs [
    A {from=MaterialExpressionWorldPosition_1, output=XYZ}
    B {from=MaterialExpressionTransformPosition_0}
   ]
  }
  MaterialExpressionSubtract_2 {
   class=MaterialExpressionSubtract
   inputs [
    A {from=MaterialExpressionMultiply_37}
    B {from=MaterialExpressionConvert_6}
   ]
  }
  MaterialExpressionSubtract_3 {
   class=MaterialExpressionSubtract
   inputs [
    A {from=MaterialExpressionConvert_17}
    B {from=MaterialExpressionConvert_2}
   ]
  }
  MaterialExpressionSubtract_4 {
   class=MaterialExpressionSubtract
   inputs [
    A {from=MaterialExpressionConvert_2}
    B {from=MaterialExpressionConvert_2}
   ]
  }
  MaterialExpressionSubtract_5 {
   class=MaterialExpressionSubtract
   inputs [
    A {from=MaterialExpressionMultiply_39}
    B {from=MaterialExpressionConvert_7}
   ]
  }
  MaterialExpressionSubtract_6 {
   class=MaterialExpressionSubtract
   inputs [
    A {from=MaterialExpressionSubtract_8}
    B {from=MaterialExpressionConvert_5}
   ]
  }
  MaterialExpressionSubtract_7 {
   class=MaterialExpressionSubtract
   inputs [
    A {from=MaterialExpressionConvert_5}
    B {from=MaterialExpressionConvert_5}
   ]
  }
  MaterialExpressionSubtract_8 {
   class=MaterialExpressionSubtract
   inputs [
    A {from=MaterialExpressionMultiply_15}
    B {from=MaterialExpressionConvert_10}
   ]
  }
  MaterialExpressionTemporalResponsivenessOutput_1 {class=MaterialExpressionTemporalResponsivenessOutput, inputs [None {from=MaterialExpressionSphereMask_0, output=""}], properties {}}
  MaterialExpressionTextureCoordinate_1 {class=MaterialExpressionTextureCoordinate}
  MaterialExpressionTextureSample_0 {
   class=MaterialExpressionTextureSample
   inputs [
    UVs {from=MaterialExpressionSubtract_2}
    Tex {from=""}
    Apply View MipBias {from=""}
   ]
   properties {
    desc=Frame
    samplerSource=SSM_Clamp_WorldGroupSettings
    texture=/BRCosmetics/Weapons/FORT_Melee/Pickaxe_SprintLuckSweat/Textures/T_SprintLuckSweat_Atlas
   }
  }
  MaterialExpressionTextureSample_1 {
   class=MaterialExpressionTextureSample
   inputs [
    UVs {from=MaterialExpressionConvert_17}
    Tex {from=""}
    Apply View MipBias {from=""}
   ]
   properties {
    desc=Moon
    samplerSource=SSM_Clamp_WorldGroupSettings
    texture=&T_SprintLuckSweat_Atlas
   }
  }
  MaterialExpressionTextureSample_2 {
   class=MaterialExpressionTextureSample
   inputs [
    UVs {from=MaterialExpressionSubtract_8}
    Tex {from=""}
    Apply View MipBias {from=""}
   ]
   properties {
    desc=BG
    samplerSource=SSM_Clamp_WorldGroupSettings
    texture=&T_SprintLuckSweat_Atlas
   }
  }
  MaterialExpressionTime_0 {class=MaterialExpressionTime}
  MaterialExpressionTime_2 {class=MaterialExpressionTime}
  MaterialExpressionTransformPosition_0 {
   class=MaterialExpressionTransformPosition
   inputs [None {from=MaterialExpressionConstant3Vector_8, output=""}]
   properties {transformType=TRANSFORMPOSSOURCE_World}
  }
  MaterialExpressionTransformPosition_1 {
   class=MaterialExpressionTransformPosition
   inputs [None {from=MaterialExpressionConstant3Vector_1, output=""}]
   properties {transformType=TRANSFORMPOSSOURCE_World}
  }
  MaterialExpressionVectorParameter_0 {
   class=MaterialExpressionVectorParameter
   properties {
    group=SprintLuckSweat
    parameterName=TextureA_Crop
    channelNames {a=SubTexMax.Y, b=SubTexMax.X, g=SubTexMin.Y, r=SubTexMin.X}
    defaultValue=<rgba 0.01 0.225 0.48 0.99>
   }
  }
  MaterialExpressionVectorParameter_1 {
   class=MaterialExpressionVectorParameter
   properties {
    group=SprintLuckSweat
    parameterName=TextureB_Crop
    channelNames {a=SubTexMax.Y, b=SubTexMax.X, g=SubTexMin.Y, r=SubTexMin.X}
    defaultValue=<rgba 0.49 0.01 0.99 0.48>
   }
  }
  MaterialExpressionVectorParameter_2 {
   class=MaterialExpressionVectorParameter
   properties {
    group=SprintLuckSweat
    parameterName=TextureC
    channelNames {a=SubTexMax.Y, b=SubTexMax.X, g=SubTexMin.Y, r=SubTexMin.X}
    defaultValue=<rgba 0.49 0.49 0.99 0.99>
   }
  }
  MaterialExpressionVectorParameter_6 {
   class=MaterialExpressionVectorParameter
   properties {group=SprintLuckSweat, parameterName=CubeMap_Darken_Color, defaultValue=<rgba 0.1 0.1 0.1 1>}
  }
  MaterialExpressionVertexNormalWS_0 {class=MaterialExpressionVertexNormalWS, inputs [], properties {}}
  MaterialExpressionViewSize_0 {class=MaterialExpressionViewSize, inputs [], properties {}}
  MaterialExpressionWorldPosition_0 {class=MaterialExpressionWorldPosition, inputs [], properties {}}
  MaterialExpressionWorldPosition_1 {class=MaterialExpressionWorldPosition, inputs [], properties {}}
 }}
