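# Parent material for the SprintLuckSweat pickaxe, with an optional animated feather layer.
# Character-parent PreFX output optionally blends with feather attributes by mask, then runs through MF_SprintLuckSweat and PostFX.
# Feathers use two-sided foliage shading, dithered alpha or optional sorted translucency, and separate front and back face color and emissive.
# Vertex color and UV masks accompany sine and grass-wind vertex motion, skylight cubemap sampling and time-of-day scaling.
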Material /BRCosmetics/Weapons/FORT_Melee/Pickaxe_SprintLuckSweat/Materials/M_SprintLuckSweat_Parent
graph {
 properties {
  bAutomaticallySetUsageInEditor=false
  bUseMaterialAttributes=true
  bUsedWithMeshDeformer=true
  bUsedWithMorphTargets=true
  bUsedWithSkeletalMesh=true
  pixelDepthOffsetMode=PDOM_Legacy
  previewMesh=/BRCosmetics/Weapons/FORT_Melee/Pickaxe_SprintLuckSweat/Meshes/SprintLuckSweat_Axe
  shadingModel=MSM_FromMaterialExpression
  usedShadingModels=Default Lit | Subsurface | Clear Coat | Subsurface Profile | Two Sided Foliage | Cloth
 }
 nodes [
  n1 = Multiply(A=CameraVectorWS()) {constB=-1}
  Add(A=@n1, B=Multiply(A=VertexNormalWS(), B=Constant() {r=0.5}))
  n2 = NamedRerouteUsage()
  n3 = GetMaterialAttributes(None=NamedRerouteUsage()) {attributeGetTypes [#1, #2, #3]}
  n4 = TextureSampleParameter2D() {parameterName=AlphaSourceTexture, texture=/BRCosmetics/Weapons/FORT_Melee/Pickaxe_SprintLuckSweat/Textures/T_SprintLuckSweat_Pickaxe_D}
  n5 = SetMaterialAttributes!(
   MaterialAttributes=@n3.MaterialAttributes
   Opacity=StaticSwitchParameter(
    True=MaterialFunctionCall!(Opacity=@n4.A) {materialFunction=/Game/Global/Materials/MaterialGym/Materials/SortedOutline/MF_SortTranslucency}.Result
    False=Constant() {r=1}
   ) {parameterName=UseTranslucency}
   Opacity Mask=MaterialFunctionCall(Alpha Threshold=@n4.A) {materialFunction=&DitherTemporalAA_Fortnite}.Result
   Shading Model=ShadingModel() {shadingModel=MSM_TwoSidedFoliage}
   World Position Offset=NamedRerouteUsage()
   Base Color=NamedRerouteUsage()
   Emissive Color=NamedRerouteUsage()
  ) {attributeSetTypes [
    #4
    #5
    #6
    #1
    #2
    #7
   ]}
  n6 = BlendMaterialAttributes(A=@n2, B=@n5, Alpha=NamedRerouteUsage())
  n7 = MaterialFunctionCall() {materialFunction=&MF_CharacterParent_2023_PreFX}
  MaterialFunctionCall(
   Material=MaterialFunctionCall!(IN MA=StaticSwitchParameter(True=@n6, False=@n2) {group=Feathers, parameterName=UseFeathers}) {materialFunction=&/MF_SprintLuckSweat}.Result
   M=@n7.M
   Pre Skinned Local Position=@n7.M
   Distance Blend=@n7.M
   Fresnel=@n7.M
  ) {materialFunction=&MF_CharacterParent_2023_PostFX}
  NamedRerouteDeclaration(None=@n7.Tech Art Mask) {name=TechArtMaskIn, nodeColor=<rgba 0 1 0.3203 1>}
  NamedRerouteDeclaration(None=ChannelMaskParameter(None=NamedRerouteUsage()) {maskChannel=Blue, parameterName=FeathersMask, defaultValue=<rgba 0 0 1 0>}.R) {name=FeathersMask, nodeColor=<rgba 1 0 0 1>}
  n8 = Power(Base=Fresnel() {exponent=2})
  n9 = Saturate(None=TwoSidedSign())
  NamedRerouteDeclaration(None=LinearInterpolate(
    A=Multiply(A=@n3.Base Color, B=@n8) {constB=0.5}
    B=NamedRerouteDeclaration(None=@n3.Base Color) {name=IN_BaseColor, nodeColor=<rgba 0.3594 1 0 1>}
    Alpha=@n9
   )) {name=Out_BaseColor, nodeColor=<rgba 0.125 0 1 1>}
  n10 = NamedRerouteUsage()
  n11 = Constant4Vector() {constant=<rgba 0.6 0.5 0.35 0.5>}
  n12 = MaterialFunctionCall(Value_Morning=@n11.R, Value_Day=@n11.R, Value_Evening=@n11.R, Value_Night=@n11.R) {materialFunction=&Scalar_Time_of_Day}
  NamedRerouteDeclaration(None=LinearInterpolate(A=Multiply(A=@n10, B=@n12.Result) {constB=0.75}, B=Multiply(A=@n10, B=@n8), Alpha=@n9) {constB=0}) {name=Out_Emissive, nodeColor=<rgba 0.125 0 1 1>}
  NamedRerouteDeclaration(None=@n7.Result) {name=MaterialIn, nodeColor=<rgba 1 0 0 1>}
  n13 = Time()
  n14 = Subtract(A=@n13, B=ConstantBiasScale(None=NamedRerouteUsage()) {bias=-0.5, scale=2})
  n15 = Sine(None=Multiply(A=@n14, B=ScalarParameter() {defaultValue=1, group=Feathers, parameterName=MainFrequency})) {period=2}
  n16 = Sine(None=Multiply(A=@n14, B=ScalarParameter() {defaultValue=0.25, group=Feathers, parameterName=SecondaryFrequency}))
  n17 = Sine(None=Multiply(A=@n13) {constB=0.1})
  n18 = NamedRerouteUsage()
  n19 = Add(A=Multiply(A=Multiply(A=@n15), B=Multiply(A=@n16)), B=Multiply(A=Sine(None=@n17), B=LinearInterpolate(Alpha=@n18) {constA=0.25}))
  n20 = Constant3Vector() {constant=<rgba 1 1 1 0>}
  n21 = Add(
   A=MaterialFunctionCall!(
    WindIntensity=ScalarParameter() {defaultValue=0.01, group=Feathers, parameterName=TertiaryIntencity}
    WindWeight=NamedRerouteUsage()
    WindSpeed=ScalarParameter() {defaultValue=0.5, group=Feathers, parameterName=TertiarySpeed}
    AdditionalWPO=Constant()
   ) {materialFunction=/Engine/Functions/Engine_MaterialFunctions01/WorldPositionOffset/SimpleGrassWind}.Result
   B=Multiply(A=Multiply(A=@n19, B=@n20), B=ScalarParameter() {defaultValue=0.15, group=Feathers, parameterName=WPOIntensity})
  )
  n22 = NamedRerouteUsage()
  NamedRerouteDeclaration(None=Add(A=Multiply(A=@n21, B=@n22), B=NamedRerouteUsage())) {name=WPO Out, nodeColor=<rgba 1 0.7734 0 1>}
  NamedRerouteDeclaration(None=@n3.World Position Offset) {name=WPO In, nodeColor=<rgba 0.4531 0 1 1>}
  n23 = ScalarParameter() {defaultValue=0.8, group=SprintLuckSweat, parameterName=SkylightEnvMapRoughness}
  n24 = NamedRerouteUsage()
  n25 = LocalPosition() {localOrigin=InstancePreSkinning}
  n26 = Constant3Vector() {constant=<rgba 1.072 7.243 25 0>}
  n27 = Constant3Vector() {constant=<rgba 1 1 -1 0>}
  n28 = SphereMask(A=@n25.XYZ, B=LinearInterpolate(A=Multiply(A=@n26, B=@n27), B=@n26, Alpha=Step(X=@n25.Z) {constX=0.5, constY=0.5})) {attenuationRadius=50, hardnessPercent=0}
  n29 = VertexInterpolator(VS=@n28)
  n30 = Max(
   A=DotProduct(A=SkyAtmosphereLightDirection(), B=VertexNormalWS())
   B=Power(Base=ShaderStageSwitch(PixelShader=@n29.PS, VertexShader=@n28)) {constExponent=3}
  )
  NamedRerouteDeclaration(None=Multiply(A=SkyLightEnvMapSample(Direction=@n1, Roughness=@n23), B=Multiply(A=@n24, B=@n30))) {name=Cubemap, nodeColor=<rgba 1 0 0.7266 1>}
  n31 = VertexColor()
  n32 = Power(Base=@n31.R, Exp=ScalarParameter() {defaultValue=16, parameterName=Exp})
  NamedRerouteDeclaration(None=@n32) {name=WPOMask, nodeColor=<rgba 0.3125 0 1 1>}
  NamedRerouteDeclaration(None=@n31.G) {name=CardIDMask, nodeColor=<rgba 0.0078 1 0 1>}
  NamedRerouteDeclaration(None=@n31.B) {name=CardsMask, nodeColor=<rgba 1 0 0.3281 1>}
  n33 = ComponentMask(None=ConstantBiasScale(None=TextureCoordinate()) {bias=-0.5, scale=2}) {r=true}
  n34 = Add(A=@n33, B=ScalarParameter() {defaultValue=0.79})
  n35 = ScalarParameter() {defaultValue=32, parameterName=Param_1}
  n36 = Power(Base=OneMinus(None=Abs(None=@n34)), Exp=@n35) {constExponent=18}
  n37 = Add(A=@n33, B=ScalarParameter() {defaultValue=0.32, parameterName=fdfg})
  n38 = Power(Base=OneMinus(None=Abs(None=@n37)), Exp=@n35) {constExponent=18}
  n39 = ScalarParameter() {defaultValue=-0.09, parameterName=egsfg}
  n40 = OneMinus(None=Abs(None=Add(A=@n33, B=@n39)))
  n41 = Saturate(None=Max(A=Max(A=@n36, B=@n38), B=Power(Base=@n40, Exp=@n35) {constExponent=18}))
  n42 = NamedRerouteUsage()
  NamedRerouteDeclaration(None=SmoothStep(Value=Multiply(A=@n41, B=@n42)) {constMin=0.2}) {name=Feather Stem masks, nodeColor=<rgba 1 0 0.6328 1>}
  ShadingModel() {shadingModel=MSM_TwoSidedFoliage}
  Subtract(A=@n32, B=NamedRerouteUsage())
  VectorParameter() {group=SprintLuckSweat, parameterName=CubeMap_Darken_Color, defaultValue=<rgba 0.1 0.1 0.1 1>}
 ]
}
