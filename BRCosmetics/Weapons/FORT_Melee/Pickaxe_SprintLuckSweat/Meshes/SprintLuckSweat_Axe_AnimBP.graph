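# Animation Blueprint for the SprintLuckSweat pickaxe that adds spring-driven secondary motion to seven feather bones using chained Anim Dynamics nodes.
# A cached input pose feeds three differently tuned chains, selected by bool blends.
# In game it uses full-strength bone-relative simulation; in the front end it uses world-space at reduced alpha, or bone-relative with a nudged MSTR_feather_1 when on a pawn.
# The update event sets FE_withpawn from bIsInFrontEnd AND bIsOnPawn.
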AnimBlueprint /BRCosmetics/Weapons/FORT_Melee/Pickaxe_SprintLuckSweat/Meshes/SprintLuckSweat_Axe_AnimBP
target_skeleton=&/SprintLuckSweat_Axe_Skeleton
graphs [
 AnimGraph {nodes [
   n1 = "Transform (Modify) Bone - Bone: MSTR_feather_1"(ComponentPose=Local To Component(LocalPose=Use cached pose 'fe/game'()), Alpha=1, Translation="0,-1.4,0", Rotation="0,0,0", Scale="1,1,1") {translationMode=BMM_Additive, translationSpace=BCS_BoneSpace, boneToModify {boneName=MSTR_feather_1}, translation {y=-1.4}}
   n2 = Anim Dynamics - Bone: feather_1!(ComponentPose=@n1, Alpha=0.32) -> (Pose) {
    angularDampingOverride=0.99
    angularSpringConstant=800
    bLinearSpring=true
    bOverrideLinearDamping=true
    bUseGravityOverride=true
    bUsePlanarLimit=false
    gravityScale=0
    linearDampingOverride=1
    linearSpringConstant=200
    simulationSpace=BoneRelative
    alphaBoolBlend {blendOption=Linear}
    boundBone {boneName=feather_1}
    componentAppliedLinearAccClamp=<3000 3000 3000>
    componentLinearAccScale=<0.28 0.1 0.15>
    componentLinearVelScale=<0.4 0.4 0.4>
    gravityOverride {z=-980}
    physicsBodyDefinitions [{
      boundBone {boneName=feather_1}
      boxExtents=<15 15 15>
      constraintSetup {angularLimitsMax=<0 0 0>, angularLimitsMin=<0 0 0>, angularTarget=<1 0 0>, linearAxesMax=<3 0 2>, linearAxesMin=<-1 -3 0>}
      localJointOffset=<0.7 0 0>
     }]
    relativeSpaceBone {boneName=b_AR_Root}
   }
   n3 = Anim Dynamics - Bone: feather_2!(ComponentPose=@n2, Alpha=0.7) -> (Pose) {
    angularDampingOverride=0.99
    angularSpringConstant=800
    bLinearSpring=true
    bOverrideLinearDamping=true
    bUseGravityOverride=true
    bUsePlanarLimit=false
    gravityScale=0
    linearDampingOverride=1
    linearSpringConstant=100
    simulationSpace=BoneRelative
    alphaBoolBlend {blendOption=Linear}
    boundBone {boneName=feather_2}
    componentAppliedLinearAccClamp=<3000 3000 3000>
    componentLinearAccScale=<0.28 0.1 0.15>
    componentLinearVelScale=<0.4 0.4 0.4>
    gravityOverride {z=-980}
    physicsBodyDefinitions [{
      boundBone {boneName=feather_2}
      boxExtents=<15 15 15>
      constraintSetup {angularLimitsMax=<0 0 0>, angularLimitsMin=<0 0 0>, angularTarget=<1 0 0>, linearAxesMax=<1 1 3>, linearAxesMin=<-2 -1 -1>}
      localJointOffset=<0.7 0 0>
     }]
    relativeSpaceBone {boneName=b_AR_Root}
   }
   n4 = Anim Dynamics - Bone: feather_3!(ComponentPose=@n3, Alpha=1) -> (Pose) {
    angularDampingOverride=0.99
    angularSpringConstant=800
    bLinearSpring=true
    bOverrideLinearDamping=true
    bUseGravityOverride=true
    bUsePlanarLimit=false
    gravityScale=0
    linearDampingOverride=1
    linearSpringConstant=150
    simulationSpace=BoneRelative
    alphaBoolBlend {blendOption=Linear}
    boundBone {boneName=feather_3}
    componentAppliedLinearAccClamp=<3000 3000 3000>
    componentLinearAccScale=<0.28 0.1 0.15>
    componentLinearVelScale=<0.4 0.4 0.4>
    gravityOverride {z=-980}
    physicsBodyDefinitions [{
      boundBone {boneName=feather_3}
      boxExtents=<15 15 15>
      constraintSetup {angularLimitsMax=<0 0 0>, angularLimitsMin=<0 0 0>, angularTarget=<1 0 0>, linearAxesMax=<2 2 2>, linearAxesMin=<-2 -2 -2>}
      localJointOffset=<0.7 0 0>
     }]
    relativeSpaceBone {boneName=b_AR_Root}
   }
   n5 = Anim Dynamics - Bone: feather_4!(ComponentPose=@n4, Alpha=1) -> (Pose) {
    angularDampingOverride=0.99
    angularSpringConstant=800
    bLinearSpring=true
    bOverrideLinearDamping=true
    bUseGravityOverride=true
    bUsePlanarLimit=false
    gravityScale=0
    linearDampingOverride=1
    linearSpringConstant=200
    simulationSpace=BoneRelative
    alphaBoolBlend {blendOption=Linear}
    boundBone {boneName=feather_4}
    componentAppliedLinearAccClamp=<3000 3000 3000>
    componentLinearAccScale=<0.28 0.1 0.15>
    componentLinearVelScale=<0.4 0.4 0.4>
    gravityOverride {z=-980}
    physicsBodyDefinitions [{
      boundBone {boneName=feather_4}
      boxExtents=<15 15 15>
      constraintSetup {angularLimitsMax=<0 0 0>, angularLimitsMin=<0 0 0>, angularTarget=<1 0 0>, linearAxesMax=<3 1 2>, linearAxesMin=<-1 -3 -4>}
      localJointOffset=<0.7 0 0>
     }]
    relativeSpaceBone {boneName=b_AR_Root}
   }
   n6 = Anim Dynamics - Bone: feather_5!(ComponentPose=@n5, Alpha=1) -> (Pose) {
    angularDampingOverride=0.99
    angularSpringConstant=800
    bLinearSpring=true
    bOverrideLinearDamping=true
    bUseGravityOverride=true
    bUsePlanarLimit=false
    gravityScale=0
    linearDampingOverride=1
    linearSpringConstant=100
    simulationSpace=BoneRelative
    alphaBoolBlend {blendOption=Linear}
    boundBone {boneName=feather_5}
    componentAppliedLinearAccClamp=<3000 3000 3000>
    componentLinearAccScale=<0.28 0.1 0.15>
    componentLinearVelScale=<0.4 0.4 0.4>
    gravityOverride {z=-980}
    physicsBodyDefinitions [{
      boundBone {boneName=feather_5}
      boxExtents=<15 15 15>
      constraintSetup {angularLimitsMax=<0 0 0>, angularLimitsMin=<0 0 0>, angularTarget=<1 0 0>, linearAxesMax=<1 2 3>, linearAxesMin=<-2 -1 -1>}
      localJointOffset=<0.7 0 0>
     }]
    relativeSpaceBone {boneName=b_AR_Root}
   }
   n7 = Anim Dynamics - Bone: feather_6!(ComponentPose=@n6, Alpha=1) -> (Pose) {
    angularDampingOverride=0.99
    angularSpringConstant=800
    bLinearSpring=true
    bOverrideLinearDamping=true
    bUseGravityOverride=true
    bUsePlanarLimit=false
    gravityScale=0
    linearDampingOverride=1
    linearSpringConstant=150
    simulationSpace=BoneRelative
    alphaBoolBlend {blendOption=Linear}
    boundBone {boneName=feather_6}
    componentAppliedLinearAccClamp=<3000 3000 3000>
    componentLinearAccScale=<0.28 0.1 0.15>
    componentLinearVelScale=<0.4 0.4 0.4>
    gravityOverride {z=-980}
    physicsBodyDefinitions [{
      boundBone {boneName=feather_6}
      boxExtents=<15 15 15>
      constraintSetup {angularLimitsMax=<0 0 0>, angularLimitsMin=<0 0 0>, angularTarget=<1 0 0>, linearAxesMax=<2 2 2>, linearAxesMin=<-2 -2 -2>}
      localJointOffset=<0.7 0 0>
     }]
    relativeSpaceBone {boneName=b_AR_Root}
   }
   n8 = Anim Dynamics - Bone: feather_7!(ComponentPose=@n7, Alpha=1) -> (Pose) {
    angularDampingOverride=0.99
    angularSpringConstant=800
    bLinearSpring=true
    bOverrideLinearDamping=true
    bUseGravityOverride=true
    bUsePlanarLimit=false
    gravityScale=0
    linearDampingOverride=1
    linearSpringConstant=100
    simulationSpace=BoneRelative
    alphaBoolBlend {blendOption=Linear}
    boundBone {boneName=feather_7}
    componentAppliedLinearAccClamp=<3000 3000 3000>
    componentLinearAccScale=<0.28 0.1 0.15>
    componentLinearVelScale=<0.4 0.4 0.4>
    gravityOverride {z=-980}
    physicsBodyDefinitions [{
      boundBone {boneName=feather_7}
      boxExtents=<15 15 15>
      constraintSetup {angularLimitsMax=<0 0 0>, angularLimitsMin=<0 0 0>, angularTarget=<1 0 0>, linearAxesMax=<1 2 3>, linearAxesMin=<-2 -1 -1>}
      localJointOffset=<0.7 0 0>
     }]
    relativeSpaceBone {boneName=b_AR_Root}
   }
   n9 = Component To Local(ComponentPose=@n8)
   n10 = Anim Dynamics - Bone: feather_1(ComponentPose=Local To Component(LocalPose=Use cached pose 'fe/game'()), Alpha=0.4) {
    angularDampingOverride=0.99
    angularSpringConstant=800
    bLinearSpring=true
    bOverrideLinearDamping=true
    bUseGravityOverride=true
    bUsePlanarLimit=false
    gravityScale=0
    linearDampingOverride=0.9999
    linearSpringConstant=200
    simulationSpace=World
    alphaBoolBlend {blendOption=Linear}
    boundBone {boneName=feather_1}
    componentAppliedLinearAccClamp=<3000 3000 3000>
    componentLinearAccScale=<0.3 0.3 0.3>
    componentLinearVelScale=<0.4 0.4 0.4>
    gravityOverride {z=-980}
    physicsBodyDefinitions [{
      boundBone {boneName=feather_1}
      boxExtents=<15 15 15>
      constraintSetup {angularLimitsMax=<0 0 0>, angularLimitsMin=<0 0 0>, angularTarget=<1 0 0>, linearAxesMax=<3 1 2>, linearAxesMin=<-1 -3 -4>}
      localJointOffset=<0 2 0>
     }]
    relativeSpaceBone {boneName=b_AR_Root}
   }
   n11 = Anim Dynamics - Bone: feather_2(ComponentPose=@n10, Alpha=0.4) {
    angularDampingOverride=0.99
    angularSpringConstant=800
    bLinearSpring=true
    bOverrideLinearDamping=true
    bUseGravityOverride=true
    bUsePlanarLimit=false
    gravityScale=0
    linearDampingOverride=0.9999
    linearSpringConstant=100
    simulationSpace=World
    alphaBoolBlend {blendOption=Linear}
    boundBone {boneName=feather_2}
    componentAppliedLinearAccClamp=<3000 3000 3000>
    componentLinearAccScale=<0.3 0.3 0.3>
    componentLinearVelScale=<0.4 0.4 0.4>
    gravityOverride {z=-980}
    physicsBodyDefinitions [{
      boundBone {boneName=feather_2}
      boxExtents=<15 15 15>
      constraintSetup {angularLimitsMax=<0 0 0>, angularLimitsMin=<0 0 0>, angularTarget=<1 0 0>, linearAxesMax=<1 2 3>, linearAxesMin=<-2 -1 -1>}
      localJointOffset=<0 2 0>
     }]
    relativeSpaceBone {boneName=b_AR_Root}
   }
   n12 = Anim Dynamics - Bone: feather_3(ComponentPose=@n11, Alpha=0.4) {
    angularDampingOverride=0.99
    angularSpringConstant=800
    bLinearSpring=true
    bOverrideLinearDamping=true
    bUseGravityOverride=true
    bUsePlanarLimit=false
    gravityScale=0
    linearDampingOverride=0.9999
    linearSpringConstant=150
    simulationSpace=World
    alphaBoolBlend {blendOption=Linear}
    boundBone {boneName=feather_3}
    componentAppliedLinearAccClamp=<3000 3000 3000>
    componentLinearAccScale=<0.3 0.3 0.3>
    componentLinearVelScale=<0.4 0.4 0.4>
    gravityOverride {z=-980}
    physicsBodyDefinitions [{
      boundBone {boneName=feather_3}
      boxExtents=<15 15 15>
      constraintSetup {angularLimitsMax=<0 0 0>, angularLimitsMin=<0 0 0>, angularTarget=<1 0 0>, linearAxesMax=<2 2 2>, linearAxesMin=<-2 -2 -2>}
      localJointOffset=<0 2 0>
     }]
    relativeSpaceBone {boneName=b_AR_Root}
   }
   n13 = Anim Dynamics - Bone: feather_4(ComponentPose=@n12, Alpha=0.4) {
    angularDampingOverride=0.99
    angularSpringConstant=800
    bLinearSpring=true
    bOverrideLinearDamping=true
    bUseGravityOverride=true
    bUsePlanarLimit=false
    gravityScale=0
    linearDampingOverride=0.9999
    linearSpringConstant=200
    simulationSpace=World
    alphaBoolBlend {blendOption=Linear}
    boundBone {boneName=feather_4}
    componentAppliedLinearAccClamp=<3000 3000 3000>
    componentLinearAccScale=<0.3 0.3 0.3>
    componentLinearVelScale=<0.4 0.4 0.4>
    gravityOverride {z=-980}
    physicsBodyDefinitions [{
      boundBone {boneName=feather_4}
      boxExtents=<15 15 15>
      constraintSetup {angularLimitsMax=<0 0 0>, angularLimitsMin=<0 0 0>, angularTarget=<1 0 0>, linearAxesMax=<3 1 2>, linearAxesMin=<-1 -3 -4>}
      localJointOffset=<0 2 0>
     }]
    relativeSpaceBone {boneName=b_AR_Root}
   }
   n14 = Anim Dynamics - Bone: feather_5(ComponentPose=@n13, Alpha=0.4) {
    angularDampingOverride=0.99
    angularSpringConstant=800
    bLinearSpring=true
    bOverrideLinearDamping=true
    bUseGravityOverride=true
    bUsePlanarLimit=false
    gravityScale=0
    linearDampingOverride=0.9999
    linearSpringConstant=100
    simulationSpace=World
    alphaBoolBlend {blendOption=Linear}
    boundBone {boneName=feather_5}
    componentAppliedLinearAccClamp=<3000 3000 3000>
    componentLinearAccScale=<0.3 0.3 0.3>
    componentLinearVelScale=<0.4 0.4 0.4>
    gravityOverride {z=-980}
    physicsBodyDefinitions [{
      boundBone {boneName=feather_5}
      boxExtents=<15 15 15>
      constraintSetup {angularLimitsMax=<0 0 0>, angularLimitsMin=<0 0 0>, angularTarget=<1 0 0>, linearAxesMax=<1 2 3>, linearAxesMin=<-2 -1 -1>}
      localJointOffset=<0 2 0>
     }]
    relativeSpaceBone {boneName=b_AR_Root}
   }
   n15 = Anim Dynamics - Bone: feather_6(ComponentPose=@n14, Alpha=0.4) {
    angularDampingOverride=0.99
    angularSpringConstant=800
    bLinearSpring=true
    bOverrideLinearDamping=true
    bUseGravityOverride=true
    bUsePlanarLimit=false
    gravityScale=0
    linearDampingOverride=0.9999
    linearSpringConstant=150
    simulationSpace=World
    alphaBoolBlend {blendOption=Linear}
    boundBone {boneName=feather_6}
    componentAppliedLinearAccClamp=<3000 3000 3000>
    componentLinearAccScale=<0.3 0.3 0.3>
    componentLinearVelScale=<0.4 0.4 0.4>
    gravityOverride {z=-980}
    physicsBodyDefinitions [{
      boundBone {boneName=feather_6}
      boxExtents=<15 15 15>
      constraintSetup {angularLimitsMax=<0 0 0>, angularLimitsMin=<0 0 0>, angularTarget=<1 0 0>, linearAxesMax=<2 2 2>, linearAxesMin=<-2 -2 -2>}
      localJointOffset=<0 2 0>
     }]
    relativeSpaceBone {boneName=b_AR_Root}
   }
   n16 = Anim Dynamics - Bone: feather_7(ComponentPose=@n15, Alpha=0.4) {
    angularDampingOverride=0.99
    angularSpringConstant=800
    bLinearSpring=true
    bOverrideLinearDamping=true
    bUseGravityOverride=true
    bUsePlanarLimit=false
    gravityScale=0
    linearDampingOverride=0.9999
    linearSpringConstant=100
    simulationSpace=World
    alphaBoolBlend {blendOption=Linear}
    boundBone {boneName=feather_7}
    componentAppliedLinearAccClamp=<3000 3000 3000>
    componentLinearAccScale=<0.3 0.3 0.3>
    componentLinearVelScale=<0.4 0.4 0.4>
    gravityOverride {z=-980}
    physicsBodyDefinitions [{
      boundBone {boneName=feather_7}
      boxExtents=<15 15 15>
      constraintSetup {angularLimitsMax=<0 0 0>, angularLimitsMin=<0 0 0>, angularTarget=<1 0 0>, linearAxesMax=<1 2 3>, linearAxesMin=<-2 -1 -1>}
      localJointOffset=<1.5 0 0>
     }]
    relativeSpaceBone {boneName=b_AR_Root}
   }
   n17 = Component To Local(ComponentPose=@n16)
   n18 = Get FE_withpawn() -> (FE_withpawn:bool)
   n19 = Anim Dynamics - Bone: feather_1(ComponentPose=Local To Component(LocalPose=Use cached pose 'fe/game'()), Alpha=1) {
    angularDampingOverride=0.99
    angularSpringConstant=800
    bLinearSpring=true
    bOverrideLinearDamping=true
    bUseGravityOverride=true
    bUsePlanarLimit=false
    gravityScale=0
    linearDampingOverride=0.9999
    linearSpringConstant=200
    simulationSpace=BoneRelative
    alphaBoolBlend {blendOption=Linear}
    boundBone {boneName=feather_1}
    componentAppliedLinearAccClamp=<3000 3000 3000>
    componentLinearAccScale=<0.3 0.3 0.3>
    componentLinearVelScale=<0.4 0.4 0.4>
    gravityOverride {z=-980}
    physicsBodyDefinitions [{
      boundBone {boneName=feather_1}
      boxExtents=<15 15 15>
      constraintSetup {angularLimitsMax=<0 0 0>, angularLimitsMin=<0 0 0>, angularTarget=<1 0 0>, linearAxesMax=<3 1 2>, linearAxesMin=<-1 -3 -4>}
      localJointOffset=<1.5 0 0>
     }]
    relativeSpaceBone {boneName=b_AR_Root}
   }
   n20 = Anim Dynamics - Bone: feather_2(ComponentPose=@n19, Alpha=1) {
    angularDampingOverride=0.99
    angularSpringConstant=800
    bLinearSpring=true
    bOverrideLinearDamping=true
    bUseGravityOverride=true
    bUsePlanarLimit=false
    gravityScale=0
    linearDampingOverride=0.9999
    linearSpringConstant=100
    simulationSpace=BoneRelative
    alphaBoolBlend {blendOption=Linear}
    boundBone {boneName=feather_2}
    componentAppliedLinearAccClamp=<3000 3000 3000>
    componentLinearAccScale=<0.3 0.3 0.3>
    componentLinearVelScale=<0.4 0.4 0.4>
    gravityOverride {z=-980}
    physicsBodyDefinitions [{
      boundBone {boneName=feather_2}
      boxExtents=<15 15 15>
      constraintSetup {angularLimitsMax=<0 0 0>, angularLimitsMin=<0 0 0>, angularTarget=<1 0 0>, linearAxesMax=<1 2 3>, linearAxesMin=<-2 -1 -1>}
      localJointOffset=<1.5 0 0>
     }]
    relativeSpaceBone {boneName=b_AR_Root}
   }
   n21 = Anim Dynamics - Bone: feather_3(ComponentPose=@n20, Alpha=1) {
    angularDampingOverride=0.99
    angularSpringConstant=800
    bLinearSpring=true
    bOverrideLinearDamping=true
    bUseGravityOverride=true
    bUsePlanarLimit=false
    gravityScale=0
    linearDampingOverride=0.9999
    linearSpringConstant=150
    simulationSpace=BoneRelative
    alphaBoolBlend {blendOption=Linear}
    boundBone {boneName=feather_3}
    componentAppliedLinearAccClamp=<3000 3000 3000>
    componentLinearAccScale=<0.3 0.3 0.3>
    componentLinearVelScale=<0.4 0.4 0.4>
    gravityOverride {z=-980}
    physicsBodyDefinitions [{
      boundBone {boneName=feather_3}
      boxExtents=<15 15 15>
      constraintSetup {angularLimitsMax=<0 0 0>, angularLimitsMin=<0 0 0>, angularTarget=<1 0 0>, linearAxesMax=<2 2 2>, linearAxesMin=<-2 -2 -2>}
      localJointOffset=<1.5 0 0>
     }]
    relativeSpaceBone {boneName=b_AR_Root}
   }
   n22 = Anim Dynamics - Bone: feather_4(ComponentPose=@n21, Alpha=1) {
    angularDampingOverride=0.99
    angularSpringConstant=800
    bLinearSpring=true
    bOverrideLinearDamping=true
    bUseGravityOverride=true
    bUsePlanarLimit=false
    gravityScale=0
    linearDampingOverride=0.9999
    linearSpringConstant=200
    simulationSpace=BoneRelative
    alphaBoolBlend {blendOption=Linear}
    boundBone {boneName=feather_4}
    componentAppliedLinearAccClamp=<3000 3000 3000>
    componentLinearAccScale=<0.3 0.3 0.3>
    componentLinearVelScale=<0.4 0.4 0.4>
    gravityOverride {z=-980}
    physicsBodyDefinitions [{
      boundBone {boneName=feather_4}
      boxExtents=<15 15 15>
      constraintSetup {angularLimitsMax=<0 0 0>, angularLimitsMin=<0 0 0>, angularTarget=<1 0 0>, linearAxesMax=<3 1 2>, linearAxesMin=<-1 -3 -4>}
      localJointOffset=<1.5 0 0>
     }]
    relativeSpaceBone {boneName=b_AR_Root}
   }
   n23 = Anim Dynamics - Bone: feather_5(ComponentPose=@n22, Alpha=1) {
    angularDampingOverride=0.99
    angularSpringConstant=800
    bLinearSpring=true
    bOverrideLinearDamping=true
    bUseGravityOverride=true
    bUsePlanarLimit=false
    gravityScale=0
    linearDampingOverride=0.9999
    linearSpringConstant=100
    simulationSpace=BoneRelative
    alphaBoolBlend {blendOption=Linear}
    boundBone {boneName=feather_5}
    componentAppliedLinearAccClamp=<3000 3000 3000>
    componentLinearAccScale=<0.3 0.3 0.3>
    componentLinearVelScale=<0.4 0.4 0.4>
    gravityOverride {z=-980}
    physicsBodyDefinitions [{
      boundBone {boneName=feather_5}
      boxExtents=<15 15 15>
      constraintSetup {angularLimitsMax=<0 0 0>, angularLimitsMin=<0 0 0>, angularTarget=<1 0 0>, linearAxesMax=<1 2 3>, linearAxesMin=<-2 -1 -1>}
      localJointOffset=<1.5 0 0>
     }]
    relativeSpaceBone {boneName=b_AR_Root}
   }
   n24 = Anim Dynamics - Bone: feather_6(ComponentPose=@n23, Alpha=1) {
    angularDampingOverride=0.99
    angularSpringConstant=800
    bLinearSpring=true
    bOverrideLinearDamping=true
    bUseGravityOverride=true
    bUsePlanarLimit=false
    gravityScale=0
    linearDampingOverride=0.9999
    linearSpringConstant=150
    simulationSpace=BoneRelative
    alphaBoolBlend {blendOption=Linear}
    boundBone {boneName=feather_6}
    componentAppliedLinearAccClamp=<3000 3000 3000>
    componentLinearAccScale=<0.3 0.3 0.3>
    componentLinearVelScale=<0.4 0.4 0.4>
    gravityOverride {z=-980}
    physicsBodyDefinitions [{
      boundBone {boneName=feather_6}
      boxExtents=<15 15 15>
      constraintSetup {angularLimitsMax=<0 0 0>, angularLimitsMin=<0 0 0>, angularTarget=<1 0 0>, linearAxesMax=<2 2 2>, linearAxesMin=<-2 -2 -2>}
      localJointOffset=<1.5 0 0>
     }]
    relativeSpaceBone {boneName=b_AR_Root}
   }
   n25 = Anim Dynamics - Bone: feather_7(ComponentPose=@n24, Alpha=1) {
    angularDampingOverride=0.99
    angularSpringConstant=800
    bLinearSpring=true
    bOverrideLinearDamping=true
    bUseGravityOverride=true
    bUsePlanarLimit=false
    gravityScale=0
    linearDampingOverride=0.9999
    linearSpringConstant=100
    simulationSpace=BoneRelative
    alphaBoolBlend {blendOption=Linear}
    boundBone {boneName=feather_7}
    componentAppliedLinearAccClamp=<3000 3000 3000>
    componentLinearAccScale=<0.3 0.3 0.3>
    componentLinearVelScale=<0.4 0.4 0.4>
    gravityOverride {z=-980}
    physicsBodyDefinitions [{
      boundBone {boneName=feather_7}
      boxExtents=<15 15 15>
      constraintSetup {angularLimitsMax=<0 0 0>, angularLimitsMin=<0 0 0>, angularTarget=<1 0 0>, linearAxesMax=<1 2 3>, linearAxesMin=<-2 -1 -1>}
      localJointOffset=<1.5 0 0>
     }]
    relativeSpaceBone {boneName=b_AR_Root}
   }
   Output Pose(Result=Blend Poses by bool(
     BlendPose_0=Blend Poses by bool(BlendPose_0=@n9, BlendPose_1=@n17, bActiveValue=@n18)
     BlendPose_1=Component To Local(ComponentPose=@n25)
     bActiveValue=Get bIsInFrontEnd()
    )) {layerGroup=None}
   AnimGraphNode_SubInput_0 = Input Pose()
   Save cached pose 'fe/game'(Pose=@AnimGraphNode_SubInput_0) {cacheName=fe/game}
  ]}
 EventGraph {nodes [
   Event BlueprintUpdateAnimation()
   > Set FE_withpawn(FE_withpawn:bool=AND Boolean(A=Get bIsInFrontEnd(), B=Get bIsOnPawn() -> (bIsOnPawn:bool))) -> (Output_Get:bool)
   TryGetPawnOwner()
  ]}
]
variables [FE_withpawn {type=boolean}]
